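# Unlit, two-sided opaque heart material for Niagara mesh particles, with a faked reflection and coloured rim glow.
# A matcap-style lookup crosses view-space position and normal, remaps the G component into both UV axes, and samples a reflection texture.
# That reflection is added to a Fresnel blend from particle colour to an edge colour.
# The sum feeds the shared MF_FX_EmissiveAdjustments function.
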Material /ShieldBubble/Assets/FX/FZ/M_AdHocSquad_Heart_Opaque_Sean
graph {
 properties {
  bAutomaticallySetUsageInEditor=false
  bUsedWithNiagaraMeshParticles=true
  shadingModel=MSM_Unlit
  translucencyLightingMode=TLM_VolumetricPerVertexDirectional
  twoSided=true
  usedShadingModels=Unlit
 }
 nodes [
  n1 = ParticleColor()
  n2 = VectorParameter() {parameterName=Edge Colour, defaultValue=<rgba 0.3632 0.0724 1 1>}
  n3 = Fresnel(ExponentIn=ScalarParameter() {defaultValue=1, group=Reflection, parameterName=Fresnel Glow Exp}) {baseReflectFraction=0}
  MaterialFunctionCall(Emissive=Add(A=NamedRerouteUsage(), B=LinearInterpolate(A=@n1.RGB, B=@n2.RGB, Alpha=@n3))) {materialFunction=&MF_FX_EmissiveAdjustments}
  n4 = TransformPosition(None=WorldPosition().XYZ) {transformSourceType=TRANSFORMPOSSOURCE_World, transformType=TRANSFORMPOSSOURCE_View}
  n5 = VertexNormalWS()
  n6 = MaterialFunctionCall(Color=CrossProduct(A=Normalize(VectorInput=@n4), B=Transform(None=@n5) {transformSourceType=TRANSFORMSOURCE_World, transformType=TRANSFORM_View})) {materialFunction=&SplitComponents}
  n7 = Multiply(A=AppendVector(A=@n6.G, B=@n6.G)) {constB=0.5}
  NamedRerouteDeclaration(None=TextureSample(UVs=Add(A=@n7) {constB=0.5}) {texture=&/T_Reflection_AdHocSquad}.RGB) {name=Refl, nodeColor=<rgba 0.7578 1 0 1>}
 ]
}
