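# Unlit translucent Niagara sprite material for a procedural diamond-shaped flash effect, with no texture samples.
# Center-scaled UVs are folded into a diamond gradient, sharpened by an Exponent parameter and multiplied by a powered soft radial mask.
# Particle color feeds emissive through a shared adjustment function.
# Opacity combines the shape with particle alpha, depth fade against scene geometry and a camera-distance fade.
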Material /ShieldBubble/Assets/FX/FZ/M_DiamondFlash
graph {
 properties {bUsedWithNiagaraSprites=true, blendMode=BLEND_Translucent, pixelDepthOffsetMode=PDOM_Legacy, shadingModel=MSM_Unlit, usedShadingModels=Unlit}
 nodes [
  n1 = ParticleColor()
  MaterialFunctionCall(Emissive=@n1.RGB) {materialFunction=&MF_FX_EmissiveAdjustments}
  n2 = MaterialFunctionCall(UVs=TextureCoordinate(), Texture Scale=Constant() {r=2.2}) {materialFunction=&ScaleUVsByCenter}
  n3 = Multiply(A=Abs(None=Subtract(A=@n2.UVs) {constB=0.5})) {constB=2}
  n4 = Subtract(A=OneMinus(None=ComponentMask(None=@n3) {r=true}), B=ComponentMask(None=@n3) {g=true})
  n5 = ScalarParameter() {defaultValue=10, parameterName=Exponent}
  n6 = Constant() {r=0.5}
  n7 = SphereMask(A=@n2.UVs, B=@n6, Radius=@n6) {hardnessPercent=0}
  n8 = Multiply(A=Multiply(A=Power(Base=@n4, Exp=@n5), B=Power(Base=@n7) {constExponent=3}))
  Multiply(
   A=Multiply(A=@n1.A, B=DepthFade(Opacity=@n8))
   B=MaterialFunctionCall(Fade Length=Constant() {r=200}, Fade Offset=Constant() {r=50}) {materialFunction=&CameraDepthFade}.Result
  )
 ]
}
