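# Niagara system for the deploy effect of a shield bubble, combining ray, flare, smoke and light pillar emitters.
# Loop-age curves shape short-lived ray meshes and flare sprites, with spawn-group masks varying flares and a secondary-colour user parameter tinting their lights.
# A smoke burst expands under drag, while a distance-culled pillar pulses with time-of-day alpha.
# Distance remaps scale rays, flares and lights; the remaining emitters carry no renderers.
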NiagaraSystem /ShieldBubble/Assets/FX/FZ/NS_ShieldBubble_FZ_DeployLW
dependencies {
 usedDataInterfaces [
  &NiagaraDataInterfaceVectorCurve
  &NiagaraDataInterfaceArrayFloat3
  &NiagaraDataInterfaceCurve
  &NiagaraDataInterfaceArrayInt32
  &NiagaraDataInterfaceArrayFloat2
 ]
 usedDynamicInputs [
  &Multiply_LinearColorByFloat
  &RandomRangeLinearColor
  &MakeVector
 ]
 usedModules [
  &self:SystemSpawnScript.SystemScriptSource.SystemScriptGraph.NiagaraNodeAssignment_0.SetVariables_#1
  &SystemState
  &self:SystemSpawnScript.SystemScriptSource.SystemScriptGraph.NiagaraNodeAssignment_2.SetVariables_#2
  &self:SystemSpawnScript.SystemScriptSource.SystemScriptGraph.NiagaraNodeAssignment_1.SetVariables_#3
 ]
}
emitters {
 BouncingHeartLoop {rendererClasses [], renderers []}
 Burst_Rays {
  emitterUpdateScript [
   &EmitterState(Life Cycle Mode=::NewEnumerator1, Loop Behavior=::NewEnumerator1, Loop Duration=0.5, UseLoopDelay=false)
   &SpawnRate(SpawnRate=120, Use Spawn Probability=false)
  ]
  particleSpawnScript [
   &PartitionParticles()
   &InitializeParticle(
    Color Mode=::NewEnumerator0
    Lifetime Max=0.3
    Lifetime Min=0.1
    Lifetime Mode=::NewEnumerator1
    Mesh Scale Max=&VectorFromCurve(
     VectorCurve=DataInterface() {propertyValues {XCurve {keys [<at 0 16>,<at 0.6 4>]}, YCurve {keys [<at 0 12>,<at 0.6 6>]}, ZCurve {keys [<at 0 12>,<at 0.6 6>]}}}
     CurveIndex=$Emitter.NormalizedLoopAge
    )
    Mesh Scale Min=&VectorFromCurve(
     VectorCurve=DataInterface() {propertyValues {XCurve {keys [<at 0 7>,<at 0.6 5>]}, YCurve {keys [<at 0 2>,<at 0.6 1>]}, ZCurve {keys [<at 0 2>,<at 0.6 1>]}}}
     CurveIndex=$Emitter.NormalizedLoopAge
    )
    Mesh Scale Mode=::NewEnumerator2
    Position Offset=<0 0 20>
    UsePositionOffset=true
    Write Material Random=true
   )
   &CameraOffset(Camera Offset Amount=35)
   &DynamicMaterialParameters(
    Index 0 Param 1=&RandomRangeFloat(Minimum=-1, Maximum=1)
    Module.Index0Param0Write=true
    Module.Index0Param1Write=false
    Module.Index0Param2Write=false
    Module.Index0Param3Write=true
   )
   &InitialMeshOrientation(
    Mesh Orientation Mode=::NewEnumerator2
    Orientation Vector=&SelectVectorFromArray(
     Direct Array Index=$Particles.PartitionParticles.Partition
     Vector Selection Array=DataInterface() {propertyValues {FloatData [<-1 0 0>, <1 0 0>]}}
    )
    Rotation=&RandomRangeVector(Minimum=<0 0 0>, Maximum=<1 0 0>)
    Use Rotation Vector=true
   )
   /SetVariables(Particles.MeshOrientation=&MultiplyQuaternion(
     Quaternion A=$Particles.MeshOrientation
     Quaternion B=&MakeQuatFromAxisAngle(
      Axis=<0 1 0>
      AngleInDegrees=&FloatFromCurve(
       FloatCurve=DataInterface() {propertyValues {Curve {keys [{tangentMode=RCTM_User, time=0, value=0}, {tangentMode=RCTM_User, time=0.5, value=1}]}}}
       CurveIndex=$Emitter.LoopedAge
       Scale Curve=90
      )
      NormalizeAxis=true
     )
    )) {isSetParameters=true}
  ]
  particleUpdateScript [
   &ParticleState()
   &ScaleMeshSize(Scale Factor=&VectorFromCurve(VectorCurve=DataInterface() {propertyValues {
       XCurve {keys [
         {arriveTangent=1.6072, interpMode=RCIM_Cubic, leaveTangent=1.6072, tangentMode=RCTM_User, time=0, value=0.2957}
         {arriveTangent=0.7143, interpMode=RCIM_Cubic, leaveTangent=0.7143, time=0.3342, value=0.9086}
         <at 0.986 1 RCIM_Cubic>
        ]}
       YCurve {keys [<at 0 1>,<at 1 1>]}
       ZCurve {keys [
         {arriveTangent=-0.9551, interpMode=RCIM_Cubic, time=0, value=1}
         {arriveTangent=-0.7266, interpMode=RCIM_Cubic, leaveTangent=-0.7266, time=0.3342, value=0.9436}
         {arriveTangent=-1.2279, interpMode=RCIM_Cubic, leaveTangent=-1.2279, tangentMode=RCTM_User, time=0.986, value=0.2836}
        ]}
      }}))
   &ScaleColor(
    Scale Alpha=&FloatFromCurve(
     FloatCurve=DataInterface() {propertyValues {Curve {keys [
         {arriveTangent=16.8724, interpMode=RCIM_Cubic, leaveTangent=16.8724, tangentMode=RCTM_User, time=0, value=0}
         {arriveTangentWeight=0.0835, interpMode=RCIM_Cubic, leaveTangentWeight=0.2634, tangentMode=RCTM_User, time=0.1, value=1}
         {arriveTangentWeight=0.1782, interpMode=RCIM_Cubic, leaveTangentWeight=0.1712, tangentMode=RCTM_User, time=0.4154, value=0.9541}
         {arriveTangent=-1.632, interpMode=RCIM_Cubic, leaveTangent=-1.632, time=0.7183, value=0.1536}
         <at 1 0 RCIM_Cubic>
        ]}}}
     Scale Curve=&FN_LOD_Distance_Remap(
      Start Distance=950
      End Distance=1500
      Minimum Output=0.005
      Maximum Output=&RandomRangeFloat(Minimum=0.05, Maximum=0.2)
      Clamp Output=true
     )
    )
    Scale RGB=&VectorFromCurve(VectorCurve=DataInterface() {propertyValues {XCurve {keys [<at 0 1>,<at 1 5>]}, YCurve {keys [<at 0 1>,<at 1 0>]}, ZCurve {keys [<at 0 0.5>,<at 0.2 0>,<at 1 50>]}}})
    ScaleA=true
    ScaleRGB=true
   )
  ]
  rendererClasses [&NiagaraMeshRendererProperties]
  renderers [{rendererClass=&NiagaraMeshRendererProperties, rendererIndex=0}]
 }
 CoreFlares001 {
  emitterUpdateScript [
   &EmitterState(Life Cycle Mode=::NewEnumerator1, Loop Behavior=::NewEnumerator1, Loop Delay=0.05, Loop Duration=0.5, UseLoopDelay=true)
   &SpawnBurst_Instantaneous(Spawn Group=1, Use Spawn Probability=false, UseLoopCountLimit=false)
   &SpawnRate(SpawnGroup=1, SpawnRate=5, Use Spawn Probability=false) {module=SpawnRate001}
   &SpawnRate(SpawnGroup=2, SpawnRate=10, Use Spawn Probability=false) {module=SpawnRate002}
  ]
  particleSpawnScript [
   /SetVariables(Particles.VisibilityTag=&SelectIntFromArray(Direct Array Index=&SpawnGroup(), Int Selection Array=DataInterface() {propertyValues {IntData [0,0,1,2]}})) {isSetParameters=true}
   &InitializeParticle(
    Color=&MaskLinearColorBySpawnGroup(Masked Linear Color=<rgba 15 0.7826 0 1>, Passthrough Linear Color=<rgba 5 2.0833 2.8442 1>, Spawn Group Mask=3)
    Lifetime=&MaskFloatBySpawnGroup(
     Masked Float=&RandomRangeFloat(Minimum=0.07, Maximum=0.1)
     Passthrough Float=&MaskFloatBySpawnGroup(Masked Float=&RandomRangeFloat(Minimum=0.08, Maximum=0.12), Passthrough Float=0.4, Spawn Group Mask=3)
     Spawn Group Mask=2
    )
    Position Offset=&MaskVectorBySpawnGroup(Masked Vector=<0 0 45>, Passthrough Vector=<0 0 20>, Spawn Group Mask=2)
    Sprite Rotation Angle=&MaskFloatBySpawnGroup(Masked Float=&RandomRangeFloat(Minimum=0, Maximum=1), Passthrough Float=0, Spawn Group Mask=3)
    Sprite Rotation Mode=::NewEnumerator1
    Sprite Size=&MaskVector2dBySpawnGroup(
     Masked Vector2D=&Vector2DFromFloat(Value=&RandomRangeFloat(Minimum=30, Maximum=60))
     Passthrough Vector2D=&Lerp_Vector2(
      A=&SelectVector2DFromArray(
       Direct Array Index=&SpawnGroup()
       Vector 2D Selection Array=DataInterface() {propertyValues {FloatData [<400 400>, <750 150>]}}
      )
      B=&SelectVector2DFromArray(
       Direct Array Index=&SpawnGroup()
       Vector 2D Selection Array=DataInterface() {propertyValues {FloatData [<750 750>, <1250 200>]}}
      )
      Alpha=&FloatFromCurve(
       FloatCurve=DataInterface() {propertyValues {Curve {keys [{arriveTangent=1.2994, interpMode=RCIM_Cubic, leaveTangent=1.2994, tangentMode=RCTM_User, time=0, value=0}, <at 0.9 1 RCIM_Cubic>]}}}
       CurveIndex=$Emitter.NormalizedLoopAge
      )
     )
     Spawn Group Mask=3
    )
    Sprite Size Mode=::NewEnumerator4
    UsePositionOffset=true
    Write Material Random=true
   )
   &DynamicMaterialParameters(
    Index 0 Param 1=0.2
    Index 0 Param 2=1
    Index 0 Param 3=1
    Module.Index0Param0Write=true
    Module.Index0Param1Write=true
    Module.Index0Param2Write=true
    Module.Index0Param3Write=false
   )
   &CameraOffset(Camera Offset Amount=&MaskFloatBySpawnGroup(Masked Float=10, Passthrough Float=50, Spawn Group Mask=3))
  ]
  particleUpdateScript [
   &ParticleState()
   &ScaleSpriteSize(
    Scale Sprite Size Mode=::NewEnumerator0
    Uniform Scale Factor=&Multiply_Float(
     A=&MaskFloatBySpawnGroup(
      Masked Float=&FloatFromCurve(FloatCurve=DataInterface() {propertyValues {Curve {keys [{tangentMode=RCTM_User, time=0, value=0.2}, {tangentMode=RCTM_User, time=1, value=1}]}}})
      Passthrough Float=&FloatFromCurve(FloatCurve=DataInterface() {propertyValues {Curve {keys [{tangentMode=RCTM_User, time=0, value=1}, {tangentMode=RCTM_User, time=1, value=0}]}}})
      Spawn Group Mask=3
     )
     B=&FN_LOD_Distance_Remap(Start Distance=750, End Distance=1500, Minimum Output=0.5, Maximum Output=1, Clamp Output=true)
    )
   )
   &ScaleColor(
    Scale Alpha=&Multiply_Float(
     A=&MaskFloatBySpawnGroup(Masked Float=1, Passthrough Float=0.7, Spawn Group Mask=3)
     B=&FN_LOD_Distance_Remap(Start Distance=750, End Distance=1500, Minimum Output=0.1, Maximum Output=1, Clamp Output=true)
    )
    Scale RGB=&VectorFromFloat(Value=&MaskFloat(Mask Source=0, Mask Parameter=?, Mask=3, Masked Float=3, Passthrough Float=10))
    ScaleA=true
    ScaleRGB=true
   )
   &ScaleColor(
    Scale Alpha=&FloatFromCurve(
     FloatCurve=DataInterface() {propertyValues {Curve {keys [
         {arriveTangent=16.8724, interpMode=RCIM_Cubic, leaveTangent=16.8724, tangentMode=RCTM_User, time=0, value=0}
         {arriveTangentWeight=0.0835, interpMode=RCIM_Cubic, leaveTangentWeight=0.2634, tangentMode=RCTM_User, time=0.1, value=1}
         {arriveTangentWeight=0.1782, interpMode=RCIM_Cubic, leaveTangentWeight=0.1712, tangentMode=RCTM_User, time=0.4154, value=0.9541}
         {arriveTangent=-1.632, interpMode=RCIM_Cubic, leaveTangent=-1.632, time=0.7183, value=0.1536}
         <at 1 0 RCIM_Cubic>
        ]}}}
     Scale Curve=&FN_LOD_Distance_Remap(Start Distance=750, End Distance=1500, Minimum Output=0, Maximum Output=0.01, Clamp Output=true)
    )
    Scale RGB=&VectorFromCurve(VectorCurve=DataInterface() {propertyValues {XCurve {keys [<at 0 1>,<at 1 5>]}, YCurve {keys [<at 0 1>,<at 1 0.2>]}, ZCurve {keys [<at 0 0.5>,<at 1 2>]}}})
    ScaleA=true
    ScaleRGB=true
   ) {module=ScaleColor001}
   /SetVariables(
    Particles.LightColor=&MakeLinearColorFromVectorAndFloat("Vector (RGB)"=&MakeVectorFromLinearColorRGB(Color=$User.SecondaryColour), "Float (Alpha)"=10)
    Particles.LightRadius=&Multiply_Float(
     A=&RandomRangeFloat(Minimum=130, Maximum=200)
     B=&FN_LOD_Distance_Remap(Start Distance=350, End Distance=1000, Minimum Output=1, Maximum Output=5, Clamp Output=true)
    )
   ) {isSetParameters=true}
  ]
  rendererClasses [&NiagaraSpriteRendererProperties, &NiagaraLightRendererProperties]
  renderers [
   {rendererClass=&NiagaraSpriteRendererProperties}
   {rendererClass=&NiagaraLightRendererProperties, rendererIndex=1}
  ]
 }
 DeployFlare {rendererClasses [], renderers []}
 Deploy_CastOff_Mesh {rendererClasses [], renderers []}
 Deploy_Flare_Mesh {rendererClasses [], renderers []}
 Deploy_Heart_Burst {rendererClasses [], renderers []}
 FN_Core_Smoke_Sprite {
  emitterUpdateScript [
   &EmitterState(Life Cycle Mode=::NewEnumerator1, Loop Behavior=::NewEnumerator1, Loop Duration=1, UseLoopDelay=false)
   &SpawnBurst_Instantaneous(Spawn Count=&RandomRangeInt(Minimum=10, Maximum=13), Use Spawn Probability=false)
  ]
  particleSpawnScript [
   /SetVariables(Particles.InitSpriteRotRate=&RandomRangeFloat(Minimum=-2, Maximum=2)) {isSetParameters=true}
   &InitializeParticle(
    Color=&MakeLinearColorFromVectorAndFloat(
     "Vector (RGB)"=&MakeVectorFromLinearColorRGB(Color=<rgba 0.526 0.526 0.526 1>)
     "Float (Alpha)"=&RandomRangeFloat(Minimum=0.3, Maximum=0.5)
    )
    Lifetime Max=0.8
    Lifetime Min=0.5
    Lifetime Mode=::NewEnumerator1
    Position Offset=<0 0 20>
    Sprite Rotation Angle=&RandomRangeFloat(Minimum=0, Maximum=360)
    Sprite Rotation Mode=::NewEnumerator0
    Sprite Size Mode=::NewEnumerator1
    Uniform Sprite Size Max=250
    Uniform Sprite Size Min=180
    UsePositionOffset=true
    Write Material Random=true
   )
   &ShapeLocation(
    Cylinder Height=&RandomRangeFloat(Minimum=60, Maximum=120)
    MaskSpawnGroup=false
    Override Seed=false
    Shape Primitive=::NewEnumerator1
    Surface Only=true
    Use Endcaps In Surface Only Mode=false
    UseOwnerScale=true
   )
   &CameraOffset(Camera Offset Amount=50)
   &AddVelocity(Module.Use Falloff Distance=false, Override Seed=false, Velocity Mode=::NewEnumerator1, Velocity Speed=350)
  ]
  particleUpdateScript [
   &Drag(Drag=&RandomRangeFloat(Minimum=0, Maximum=6), Use Linear Drag=true, Use Rotational Drag=true)
   &SolveForcesAndVelocity(Clamp Velocity=false, Limit Acceleration=false)
   &ParticleState()
   &SpriteRotationRate(Rotation Rate=&Multiply_Float(
     A=$Particles.InitSpriteRotRate
     B=&FloatFromCurve(
      FloatCurve=DataInterface() {propertyValues {Curve {keys [
          {arriveTangent=-4.0107, interpMode=RCIM_Cubic, leaveTangent=-4.0107, tangentMode=RCTM_User, time=-0.0046, value=1.0113}
          {arriveTangent=-0.2496, interpMode=RCIM_Cubic, leaveTangent=-0.2496, tangentMode=RCTM_User, time=0.5, value=0.16}
         ]}}}
      CurveIndex=$Emitter.Age
      Scale Curve=150
     )
    ))
   &ScaleColor(
    Scale Alpha=&FloatFromCurve(
     FloatCurve=DataInterface() {propertyValues {Curve {keys [{interpMode=RCIM_Cubic, tangentMode=RCTM_User, time=0.6, value=1}, <at 1 0>]}}}
     Scale Curve=2
    )
    ScaleA=true
    ScaleRGB=true
   )
   &ScaleSpriteSize(Scale Factor=&Vector2DFromFloat(Value=&FloatFromCurve(FloatCurve=DataInterface() {propertyValues {Curve {keys [
          {arriveTangent=8.1876, interpMode=RCIM_Cubic, leaveTangent=5.3162, tangentMode=RCTM_User, time=0, value=0.15}
          {arriveTangent=1.0025, interpMode=RCIM_Cubic, leaveTangent=1.0025, tangentMode=RCTM_User, time=0.161, value=0.7059}
          {arriveTangent=0.1638, interpMode=RCIM_Cubic, leaveTangent=0.1638, tangentMode=RCTM_User, time=1.0463, value=1}
         ]}}})))
   &DynamicMaterialParameters(
    Index 0 Param 1=&Add_Float(
     A=$Particles.RandomFloat
     B=&FloatFromCurve(FloatCurve=DataInterface() {propertyValues {Curve {keys [
          {arriveTangent=2.7533, interpMode=RCIM_Cubic, leaveTangent=2.847, tangentMode=RCTM_Break, time=0, value=0}
          {arriveTangent=1.3581, interpMode=RCIM_Cubic, leaveTangent=1.3581, tangentMode=RCTM_User, time=0.2558, value=0.5125}
          {arriveTangent=0.0529, interpMode=RCIM_Cubic, leaveTangent=0.2021, tangentMode=RCTM_User, time=1, value=1}
         ]}}})
    )
    Index 0 Param 3=&FloatFromCurve(FloatCurve=DataInterface() {propertyValues {Curve {keys [{arriveTangent=-53.9199, time=0, value=0}, {leaveTangent=0.2723, time=0.25, value=0}]}}})
    Index 0 Param 4=&FloatFromCurve(FloatCurve=DataInterface() {propertyValues {Curve {keys [<at 0 1>,<at 1 0>]}}})
    Index 1 Param 1=&FloatFromCurve(FloatCurve=DataInterface() {propertyValues {Curve {keys [<at 0 0>, {arriveTangent=4.2257, interpMode=RCIM_Cubic, leaveTangent=4.2257, tangentMode=RCTM_User, time=0.25, value=0}]}}})
    Index 1 Param 2=&FloatFromCurve(FloatCurve=DataInterface() {propertyValues {Curve {keys [<at 0 0>]}}})
    Index 1 Param 3=&FloatFromCurve(FloatCurve=DataInterface() {propertyValues {Curve {keys [{interpMode=RCIM_Cubic, leaveTangent=6.3917, tangentMode=RCTM_Break, time=0, value=0}]}}})
    Index 1 Param 4=&FloatFromCurve(FloatCurve=DataInterface() {propertyValues {Curve {keys [{arriveTangent=-0.136, leaveTangent=-0.136, time=0, value=1}]}}})
    Index 2 Param 1=&FloatFromCurve(FloatCurve=DataInterface() {propertyValues {Curve {keys [
         {arriveTangent=2.7638, interpMode=RCIM_Cubic, leaveTangent=2.7638, tangentMode=RCTM_User, time=0, value=0}
         {arriveTangent=1, interpMode=RCIM_Cubic, leaveTangent=1, time=0.4, value=0.75}
         {leaveTangent=-0.0719, time=1, value=1}
        ]}}})
    Index 2 Param 2=&FloatFromCurve(
     FloatCurve=DataInterface() {propertyValues {Curve {keys [
         {arriveTangent=1.9056, time=0, value=0.2}
         {arriveTangent=0.0259, interpMode=RCIM_Cubic, leaveTangent=0.0259, tangentMode=RCTM_User, time=1, value=1.4}
        ]}}}
     Scale Curve=&RandomRangeFloat(Minimum=0.5, Maximum=1)
    )
    Index 2 Param 3=&FloatFromCurve(FloatCurve=DataInterface() {propertyValues {Curve {keys [
         {interpMode=RCIM_Cubic, leaveTangent=3.024, tangentMode=RCTM_Break, time=0, value=1}
         {arriveTangent=1, interpMode=RCIM_Cubic, leaveTangent=1, time=0.45, value=1.8}
         {interpMode=RCIM_Cubic, leaveTangent=11.1233, tangentMode=RCTM_Break, time=1, value=2}
        ]}}})
    Index 2 Param 4=&FloatFromCurve(
     FloatCurve=DataInterface() {propertyValues {Curve {keys [
         {arriveTangent=-0.6867, interpMode=RCIM_Cubic, leaveTangent=-0.6867, tangentMode=RCTM_User, time=0, value=1}
         {interpMode=RCIM_Cubic, leaveTangent=-1.2956, tangentMode=RCTM_Break, time=1, value=0.75}
        ]}}}
     Scale Curve=1.6
    )
    Module.Param0WriteEnabled=true
    Module.Param0WriteEnabled1=true
    Module.Param0WriteEnabled2=true
    Module.Param1WriteEnabled=false
    Module.Param1WriteEnabled1=true
    Module.Param1WriteEnabled2=true
    Module.Param2WriteEnabled=true
    Module.Param2WriteEnabled1=true
    Module.Param2WriteEnabled2=true
    Module.Param3WriteEnabled=true
    Module.Param3WriteEnabled1=true
    Module.Param3WriteEnabled2=true
    Write Parameter Index 1=true
    Write Parameter Index 2=true
   )
  ]
  rendererClasses [&NiagaraSpriteRendererProperties]
  renderers [
   {rendererClass=&NiagaraSpriteRendererProperties}
   {rendererClass=&NiagaraSpriteRendererProperties, rendererIndex=1}
  ]
 }
 HeartLoopCentre {rendererClasses [], renderers []}
 LightPillar {
  emitterUpdateScript [
   &EmitterState(
    Enable Distance Culling=true
    Life Cycle Mode=::NewEnumerator1
    Loop Behavior=::NewEnumerator1
    Loop Delay=0.15
    Loop Duration Mode=::NewEnumerator1
    Min Distance Response=::SleepAndLetParticlesFinish
    MinDistance=1000
    Module.bUseMaxDistance=false
    Module.bUseMinDistance=true
    Scalability Mode=::NewEnumerator1
    UseLoopDelay=false
   )
   &SpawnBurst_Instantaneous(Spawn Count=1, Use Spawn Probability=false)
  ]
  particleSpawnScript [
   &InitializeParticle(
    Color Mode=::NewEnumerator0
    Lifetime=2.5
    Mesh Scale=<8 3 30>
    Mesh Scale Mode=::NewEnumerator4
    UsePositionOffset=false
    Write Material Random=true
   )
   &InitialMeshOrientation(Orientation Vector=<0 0 1>, Use Orientation Vector=true, Use Rotation Vector=false)
   &ShapeLocation(MaskSpawnGroup=false, Offset=<0 0 -30>, Offset Mode=::NewEnumerator0, Override Seed=false, Sphere Radius=1, UseOwnerScale=true)
  ]
  particleUpdateScript [
   &ParticleState(Kill Particles When Lifetime Has Elapsed=false, Loop Particles Lifetime=true)
   &ScaleColor(
    Scale Alpha=&Sine(
     Normalized Angle=$Emitter.Age
     Period=3
     Scale=0.05
     Bias=&FNTimeOfDayScalar(Night Scale=0.175, Morning Scale=0.2, Day Scale=0.225, Evening Scale=0.2)
    )
    ScaleA=true
    ScaleRGB=false
   )
   &ScaleMeshSize(Scale Factor=&VectorFromCurve(
     VectorCurve=DataInterface() {propertyValues {
       XCurve {keys [{interpMode=RCIM_Cubic, tangentMode=RCTM_User, time=0, value=0}, {interpMode=RCIM_Cubic, tangentMode=RCTM_User, time=0.3, value=1}]}
       YCurve {keys [{interpMode=RCIM_Cubic, tangentMode=RCTM_User, time=0, value=0}, {interpMode=RCIM_Cubic, tangentMode=RCTM_User, time=0.3, value=1}]}
       ZCurve {keys [{interpMode=RCIM_Cubic, tangentMode=RCTM_User, time=0, value=0}, {interpMode=RCIM_Cubic, tangentMode=RCTM_User, time=0.3, value=1}]}
      }}
     CurveIndex=$Emitter.Age
    ))
   &SolveForcesAndVelocity(Clamp Velocity=false, Limit Acceleration=false)
  ]
  rendererClasses [&NiagaraMeshRendererProperties]
  renderers [{rendererClass=&NiagaraMeshRendererProperties, rendererIndex=0}]
 }
 OblongFlare_Looping {rendererClasses [], renderers []}
 SquFlare_Looping {rendererClasses [], renderers []}
}
summary {
 systemName=NS_ShieldBubble_FZ_DeployLW
 emitters [
  {emitterName=Burst_Rays, rendererClasses [&NiagaraMeshRendererProperties]}
  {emitterName=DeployFlare, rendererClasses []}
  {emitterName=Deploy_Flare_Mesh, rendererClasses []}
  {emitterName=Deploy_CastOff_Mesh, rendererClasses []}
  {emitterName=Deploy_Heart_Burst, rendererClasses []}
  {emitterName=SquFlare_Looping, rendererClasses []}
  {emitterName=OblongFlare_Looping, rendererClasses []}
  {emitterName=FN_Core_Smoke_Sprite, rendererClasses [&NiagaraSpriteRendererProperties]}
  {emitterName=BouncingHeartLoop, rendererClasses []}
  {emitterName=CoreFlares001, rendererClasses [&NiagaraSpriteRendererProperties, &NiagaraLightRendererProperties]}
  {emitterName=LightPillar, rendererClasses [&NiagaraMeshRendererProperties]}
  {emitterName=HeartLoopCentre, rendererClasses []}
 ]
 userVariables [
  User.CentreHeight {type=NiagaraFloat, defaultValue=0.315, description=""}
  User.PrimaryColour {type=LinearColor, description="", defaultValue {struct=LinearColor, value=<rgba 1 0 0.8614 1>}}
  User.SecondaryColour {type=LinearColor, description="", defaultValue {struct=LinearColor, value=<rgba 0.2494 0 1 1>}}
  User.ShieldSize {type=NiagaraFloat, defaultValue=1250, description=""}
  User.SwirlAmount {type=NiagaraFloat, defaultValue=0.25, description=""}
 ]
}
