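# Niagara system for a shield bubble's idle swirl: one GPU emitter of dense glowing sprites in six rotating spiral arms.
# A random SwirlPosition drives torus radius, twist, height, noise and a yellow-to-purple curve, scaled by User size, height and swirl parameters.
# Local camera distance shrinks nearby sprites and raises a BlurSize attribute, with two sprite renderers attached.
# User Primary and Secondary colours are declared but unreferenced in the shown modules.
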NiagaraSystem /ShieldBubble/Assets/FX/FZ/NS_ShieldBubble_FZ_IdleSwirl
dependencies {
 usedDataInterfaces [
  &NiagaraDataInterfaceCurve
  &NiagaraDataInterfaceColorCurve
  &NiagaraDataInterfaceArrayFloat
  &NiagaraDataInterfaceCamera
 ]
 usedDynamicInputs []
 usedModules [&SystemState]
}
emitters {SwirlingGlows {
  simTarget=GPUComputeSim
  emitterUpdateScript [
   &EmitterState(
    Enable Distance Culling=true
    Life Cycle Mode=::NewEnumerator1
    Loop Behavior=::NewEnumerator1
    Loop Delay=0.4
    Loop Duration Mode=::NewEnumerator1
    MaxDistance=3000
    Module.bUseMaxDistance=true
    Module.bUseMinDistance=false
    Reset Age On Awaken=true
    Scalability Mode=::NewEnumerator1
    UseLoopDelay=true
   )
   &SpawnRate(SpawnRate=2750, Use Spawn Probability=false)
   &SpawnBurst_Instantaneous(Spawn Count=750, Use Spawn Probability=false, UseLoopCountLimit=false)
  ]
  particleSpawnScript [
   /SetVariables(Particles.SwirlPosition:NiagaraFloat=&Power(a=&RandomRangeFloat(Minimum=0, Maximum=1), b=0.45)) {isSetParameters=true}
   &InitializeParticle(
    Color=&ColorFromCurve(
     ColorCurve=DataInterface() {propertyValues {
       AlphaCurve {keys [<at 0 1 RCIM_Cubic>,<at 1 0 RCIM_Cubic>]}
       BlueCurve {keys [<at 0 0 RCIM_Cubic>,<at 0.3752 0.087>,<at 0.9133 1 RCIM_Cubic>]}
       GreenCurve {keys [<at 0 0.8188 RCIM_Cubic>,<at 0.3752 0.2697>,<at 0.9133 0 RCIM_Cubic>]}
       RedCurve {keys [<at 0 1 RCIM_Cubic>,<at 0.3752 0.7956>,<at 0.9133 0.6946 RCIM_Cubic>]}
      }}
     CurveIndex=$Particles.SwirlPosition
    )
    Lifetime Max=3
    Lifetime Mode=::NewEnumerator1
    Sprite Size Mode=::NewEnumerator1
    Uniform Sprite Size Max=8
    UsePositionOffset=false
    Write Material Random=true
   )
   &ShapeLocation(
    Handle Radius=5
    Large Radius=&Lerp_Float(A=0, B=&Subtract_Float(A=$User.ShieldSize, B=5), Alpha=$Particles.SwirlPosition)
    MaskSpawnGroup=false
    Offset=<0 0 50>
    Offset Mode=::NewEnumerator0
    Rotation Angle:NiagaraFloat=&Lerp_Float(
     A=$User.SwirlAmount
     B=0
     Alpha=&Add_Float(A=$Particles.SwirlPosition, B=&Multiply_Float(A=$Emitter.Age, B=-0.035))
    )
    Rotation Axis=<0 0 1>
    Rotation Mode=::NewEnumerator1
    Shape Primitive=::NewEnumerator3
    Torus Distribution Mode:ENiagaraTorusDistributionMode=::NewEnumerator1
    U Position=&Add_Float(
     A=&SelectFloatFromArray(
      Array Sampling Mode=::NewEnumerator0
      Direct Array Index=?
      Float Selection Array=DataInterface() {propertyValues {FloatData [0,0.16,0.32,0.48,0.64,0.84]}}
      Random Seed=0
     )
     B=&RandomRangeFloat(Minimum=-0.02, Maximum=0.02)
    )
    UseOwnerScale=true
    V Position=&RandomRangeFloat(Minimum=0, Maximum=1)
   )
   /SetVariables(Particles.Position=&AddVectorToPosition(
     Position=$Particles.Position
     Vector=&Lerp_Vector(
      A=&RandomRangeVector(
       Minimum=<0 0 0>
       Maximum=&MakeVector(X=0, Y=0, Z=&Multiply_Float(A=$User.ShieldSize, B=$User.CentreHeight))
      )
      B=&MakeVector(X=0, Y=0, Z=&RandomRangeFloat(Minimum=0, Maximum=50))
      Alpha=&Power(a=$Particles.SwirlPosition, b=3)
     )
    )) {isSetParameters=true}
  ]
  particleUpdateScript [
   &ParticleState()
   &CurlNoiseForce(
    Bias Noise=false
    Noise Frequency=5
    Noise Strength=&Lerp_Float(A=5, B=50, Alpha=$Particles.SwirlPosition)
    Pan Noise=false
   )
   &SolveForcesAndVelocity(Clamp Velocity=false, Limit Acceleration=false)
   /SetVariables(Particles.Position=/Niagara/DynamicInputs/Vectors/RotateVector(
     Vector To Rotate:Vector3f=Linked() {linkedVariable {name=Particles.Position, type=NiagaraPosition}}
     Yaw:NiagaraFloat=3.75
     Pitch:NiagaraFloat=0
     Roll:NiagaraFloat=0
     Rotation Rate:NiagaraFloat=$Engine.DeltaTime
    )) {isSetParameters=true}
   /SetVariables(Particles.DistanceLerp:NiagaraFloat=&NormalizeDistanceRange(
     Start Position:NiagaraPosition=$Particles.Position
     End Position:NiagaraPosition=&ReturnCameraPosition(Coordinate Space:ENiagaraCoordinateSpace=::Local, Camera Query:NiagaraDataInterfaceCamera=DataInterface())
     Distance:NiagaraFloat=450
     Invert Normalized Range:NiagaraBool=false
    )) {isSetParameters=true}
   &ScaleColor(
    Scale Alpha=&FloatFromCurve(FloatCurve=DataInterface() {propertyValues {Curve {keys [<at 0 1>,<at 1 0>]}}})
    Scale RGB=&VectorFromFloat(Value=&RandomRangeFloat(Minimum=1, Maximum=3))
    ScaleA=true
    ScaleRGB=true
   )
   &ScaleSpriteSize(
    Scale Sprite Size Mode=::NewEnumerator0
    Uniform Scale Factor=&Lerp_Float(
     A=0
     B=&Sine(
      Normalized Angle=&Add_Float(A=$Particles.Age, B=&RandomRangeFloat(Minimum=0, Maximum=2.5))
      Period=&RandomRangeFloat(Minimum=1, Maximum=2.5)
      Scale=0.8
      Bias=1.15
     )
     Alpha=&Power(a=$Particles.DistanceLerp, b=6)
    )
   )
   /SetVariables(Particles.BlurSize:Vector2f=&Vector2DFromFloat(Value=&Lerp_Float(A=20, B=0, Alpha=&Power(a=$Particles.DistanceLerp, b=0.5)))) {isSetParameters=true}
  ]
  rendererClasses [&NiagaraSpriteRendererProperties]
  renderers [
   {rendererClass=&NiagaraSpriteRendererProperties}
   {rendererClass=&NiagaraSpriteRendererProperties, rendererIndex=1}
  ]
 }}
summary {
 systemName=NS_ShieldBubble_FZ_IdleSwirl
 emitters [{emitterName=SwirlingGlows, simTarget=GPUComputeSim, rendererClasses [&NiagaraSpriteRendererProperties]}]
 userVariables [
  User.CentreHeight {type=NiagaraFloat, defaultValue=0.315, description=""}
  User.PrimaryColour {type=LinearColor, description="", defaultValue {struct=LinearColor, value=<rgba 1 0 0.8614 1>}}
  User.SecondaryColour {type=LinearColor, description="", defaultValue {struct=LinearColor, value=<rgba 0.2494 0 1 1>}}
  User.ShieldSize {type=NiagaraFloat, defaultValue=1250, description=""}
  User.SwirlAmount {type=NiagaraFloat, defaultValue=0.25, description=""}
 ]
}
