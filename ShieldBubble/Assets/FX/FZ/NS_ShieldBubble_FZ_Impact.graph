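# Niagara system for the shield bubble FZ impact effect, judging by its name and emitters.
# It layers star mesh, ripple, small sparks, burst wave, core flare and light emitters.
# User parameters expose primary and secondary colours (purple by default), grid size and range, and impact hole amount and range.
# The system spawn script sets variables alongside SystemState, and a distance LOD remap dynamic input is among its dependencies.
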NiagaraSystem /ShieldBubble/Assets/FX/FZ/NS_ShieldBubble_FZ_Impact
dependencies {
 usedDataInterfaces []
 usedDynamicInputs [&Vector2DFromFloat, &FN_LOD_Distance_Remap]
 usedModules [
  &self:SystemSpawnScript.SystemScriptSource.SystemScriptGraph.NiagaraNodeAssignment_0.SetVariables_#1
  &SystemState
 ]
}
emitters {
 BurstWave {rendererClasses [], renderers []}
 CoreFlare {rendererClasses [], renderers []}
 Impact_Star_Mesh {rendererClasses [], renderers []}
 Light_LW {rendererClasses [], renderers []}
 Ripple {rendererClasses [], renderers []}
 Sparks_Small {rendererClasses [], renderers []}
}
summary {
 systemName=NS_ShieldBubble_FZ_Impact
 emitters [
  {emitterName=Impact_Star_Mesh, rendererClasses []}
  {emitterName=Ripple, rendererClasses []}
  {emitterName=Sparks_Small, rendererClasses []}
  {emitterName=BurstWave, rendererClasses []}
  {emitterName=CoreFlare, rendererClasses []}
  {emitterName=Light_LW, rendererClasses []}
 ]
 userVariables [
  User.GridRange {type=NiagaraFloat, defaultValue=50, description=""}
  User.GridSize {type=NiagaraFloat, defaultValue=1, description=""}
  User.ImpactHoleAmount {type=NiagaraInt32, defaultValue=1}
  User.ImpactHoleRange {type=NiagaraFloat, defaultValue=0, description=""}
  User.PrimaryColour {type=LinearColor, description="", defaultValue {struct=LinearColor, value=<rgba 0.4499 0.0625 0.7656 1>}}
  User.SecondaryColour {type=LinearColor, description="", defaultValue {struct=LinearColor, value=<rgba 0.4192 0.0576 0.4219 1>}}
 ]
}
