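# Niagara effect for the FZ variant of the shield bubble, apparently played when its time runs out.
# It combines four emitters: CoreFlare, Mesh_Castoff, Sparks_Small and SparkleStationary.
# User parameters expose primary and secondary colours, shield size and an integer stage.
# System modules include SystemState and two variable assignments, with dynamic inputs for colour and float multiplication, float-to-vector conversion and distance-based LOD remapping.
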NiagaraSystem /ShieldBubble/Assets/FX/FZ/NS_ShieldBubble_FZ_TimesUp
dependencies {
 usedDataInterfaces []
 usedDynamicInputs [
  &Vector2DFromFloat
  &FN_LOD_Distance_Remap
  &Multiply_LinearColorByFloat
  &VectorFromFloat
  &Multiply_Float
 ]
 usedModules [
  &self:SystemSpawnScript.SystemScriptSource.SystemScriptGraph.NiagaraNodeAssignment_0.SetVariables_#1
  &self:SystemSpawnScript.SystemScriptSource.SystemScriptGraph.NiagaraNodeAssignment_1.SetVariables_#2
  &SystemState
 ]
}
emitters {
 CoreFlare {rendererClasses [], renderers []}
 Mesh_Castoff {rendererClasses [], renderers []}
 SparkleStationary {rendererClasses [], renderers []}
 Sparks_Small {rendererClasses [], renderers []}
}
summary {
 systemName=NS_ShieldBubble_FZ_TimesUp
 emitters [
  {emitterName=CoreFlare, rendererClasses []}
  {emitterName=Mesh_Castoff, rendererClasses []}
  {emitterName=Sparks_Small, rendererClasses []}
  {emitterName=SparkleStationary, rendererClasses []}
 ]
 userVariables [
  User.PrimaryColour {type=LinearColor, description="", defaultValue {struct=LinearColor, value=<rgba 0.8353 0.0148 1 1>}}
  User.SecondaryColour {type=LinearColor, description="", defaultValue {struct=LinearColor, value=<rgba 1 0.6786 0.3408 1>}}
  User.ShieldSize {type=NiagaraFloat, defaultValue=1250, description=""}
  User.Stage {type=NiagaraInt32, defaultValue=2}
 ]
}
